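# Physics asset for the Hang Nine backpack cosmetic, providing capsule bodies and constraints for its dangling ropes and shark piece.
# Rope chains run from base bodies on the pack with twist locked and per-joint swing limits; left and center chains each close a loop at their rope08 body.
# Both shark fins are angularly locked to the shark root, while that root and all rope bases have free angular motion relative to the pack.
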PhysicsAsset /BRCosmetics/Accessories/FORT_Backpacks/Backpack_F_MED_HangNine/Meshes/F_MED_HangNine_Pack_Physics
bodies {
 base_rope01_cr [{shapeType=Capsule, length=0, radius=0.5, shapeName=base_rope01_cr_capsule, rotation=<rot 90 0 0>}]
 base_rope01_lf [{shapeType=Capsule, length=0, radius=0.5, shapeName=base_rope01_lf_capsule}]
 base_rope01_rt [{shapeType=Capsule, length=0.5, radius=2.5, shapeName=dyn_rope01_rt_capsule, center=<2.1602 0 0>, rotation=<rot 90 0 0>}]
 base_rope06_cr [{shapeType=Capsule, length=2.5, radius=0.5, shapeName=dyn_rope06_cr_capsule, center=<1.8 0 0>, rotation=<rot 90 0 0>}]
 base_rope06_lf [{shapeType=Capsule, length=0, radius=0.5}]
 dyn_rope01_cr [{shapeType=Capsule, length=5.8, radius=0.5, shapeName=dyn_rope01_cr_capsule, center=<3.4 0 0>, rotation=<rot 90 0 0>}]
 dyn_rope01_lf [{shapeType=Capsule, length=5.1, radius=0.5, shapeName=base_rope01_lf_capsule, center=<2.9 0 0>, rotation=<rot 90 0 0>}]
 dyn_rope01_rt [{shapeType=Capsule, length=0.5, radius=2.5, shapeName=dyn_rope01_rt_capsule, center=<2.1602 0 0>, rotation=<rot 90 0 0>}]
 dyn_rope02_cr [{shapeType=Capsule, length=4.4, radius=0.5, shapeName=dyn_rope02_cr_capsule, center=<2.75 0 0>, rotation=<rot 90 0 0>}]
 dyn_rope02_lf [{shapeType=Capsule, length=4.9, radius=0.5, shapeName=base_rope01_lf_capsule, center=<2.9 0 0>, rotation=<rot 90 0 0>}]
 dyn_rope03_lf [{shapeType=Capsule, length=4.9, radius=0.5, shapeName=base_rope01_lf_capsule, center=<3 0 0>, rotation=<rot 90 0 0>}]
 dyn_rope04_cr [{shapeType=Capsule, length=4.3, radius=0.5, shapeName=dyn_rope04_cr_capsule, center=<2.7 0 0>, rotation=<rot 90 0 0>}]
 dyn_rope04_lf [{shapeType=Capsule, length=2.6, radius=0.505, shapeName=dyn_rope04_lf_capsule, center=<1.88 -0.2597 0>, rotation=<rot 90 -6.7305 0>}]
 dyn_rope05_cr [{
   shapeType=Capsule
   length=2.15
   radius=0.5
   shapeName=dyn_rope05_cr_capsule
   center=<1.4307 0 -0.1666>
   rotation=<rot 83.3989 -0.0021 -0.0022>
  }]
 dyn_rope06_cr [{shapeType=Capsule, length=2.5, radius=0.5, shapeName=dyn_rope06_cr_capsule, center=<1.8 0 0>, rotation=<rot 90 0 0>}]
 dyn_rope06_lf [{shapeType=Capsule, length=5.1, radius=0.5, shapeName=dyn_rope06_lf_capsule, center=<2.9956 0 0>, rotation=<rot 90 -5.7574 0>}]
 dyn_rope07_cr [{shapeType=Capsule, length=2.9, radius=0.5, shapeName=dyn_rope07_cr_capsule, center=<1.85 0 0>, rotation=<rot 90 0 0>}]
 dyn_rope08_cr [{shapeType=Capsule, length=3, radius=0.505, shapeName=dyn_rope08_cr_capsule, center=<1.8846 0.5364 0>, rotation=<rot 90 10.2896 0>}]
 dyn_rope08_lf [{shapeType=Capsule, length=2.3, radius=0.5, shapeName=dyn_rope08_lf_capsule, center=<1.5599 -0.4159 0>, rotation=<rot 90 -17.8612 0>}]
 dyn_sharkFin_lf [{
   shapeType=Capsule
   length=5.5
   radius=2
   shapeName=dyn_sharkFin_lf_capsule
   center=<3.4742 2.5882 1.2428>
   rotation=<rot 12.0313 -65.6093 71.0058>
  }]
 dyn_sharkFin_rt [{
   shapeType=Capsule
   length=5.5
   radius=2
   shapeName=dyn_sharkFin_rt_capsule
   center=<3.0578 2.3299 -1.1516>
   rotation=<rot 8.8037 -58.2214 113.6735>
  }]
 dyn_sharkRoot_cr [{shapeType=Capsule, length=1, radius=9.9, shapeName=dyn_sharkRoot_cr_capsule, center=<4.3782 0 0>}]
 pack [
  {shapeType=Capsule, length=16, radius=5, shapeName=pack_capsule, center=<0 7.6204 -14.6811>, rotation=<rot 90 0 0>}
  {shapeType=Capsule, length=16, radius=5, shapeName=pack_capsule, center=<0 6.7475 -21.2147>, rotation=<rot 90 0 0>}
 ]
}
constraints [
 {bone1Name=dyn_rope01_lf, bone2Name=base_rope01_lf, swing1LimitDegrees=5, swing2LimitDegrees=7, twistLimitDegrees=5, twistMotion=Locked}
 {bone1Name=dyn_rope02_lf, bone2Name=dyn_rope01_lf, swing1LimitDegrees=13, swing2LimitDegrees=20, twistLimitDegrees=5, twistMotion=Locked}
 {bone1Name=dyn_rope03_lf, bone2Name=dyn_rope02_lf, swing1LimitDegrees=55, swing2LimitDegrees=55, twistLimitDegrees=5, twistMotion=Locked}
 {bone1Name=dyn_rope06_cr, bone2Name=base_rope06_cr, swing1LimitDegrees=7, swing2LimitDegrees=40, twistLimitDegrees=5, twistMotion=Locked}
 {bone1Name=dyn_rope07_cr, bone2Name=dyn_rope06_cr, swing1LimitDegrees=7, swing2LimitDegrees=28, twistLimitDegrees=5, twistMotion=Locked}
 {bone1Name=dyn_rope05_cr, bone2Name=dyn_rope04_cr, swing1LimitDegrees=8, swing2LimitDegrees=8, twistLimitDegrees=5, twistMotion=Locked}
 {bone1Name=dyn_rope01_cr, bone2Name=base_rope01_cr, swing1LimitDegrees=5, swing2LimitDegrees=5, twistLimitDegrees=5, twistMotion=Locked}
 {bone1Name=dyn_rope02_cr, bone2Name=dyn_rope01_cr, swing1LimitDegrees=9, swing2LimitDegrees=12, twistLimitDegrees=5, twistMotion=Locked}
 {bone1Name=dyn_rope04_cr, bone2Name=dyn_rope02_cr, swing1LimitDegrees=20, swing2LimitDegrees=25, twistLimitDegrees=5, twistMotion=Locked}
 {bone1Name=dyn_rope06_lf, bone2Name=base_rope06_lf, swing1LimitDegrees=55, swing2LimitDegrees=55, twistLimitDegrees=10, twistMotion=Locked}
 {bone1Name=dyn_rope08_cr, bone2Name=dyn_rope07_cr, swing1LimitDegrees=7, swing2LimitDegrees=10, twistLimitDegrees=45, twistMotion=Locked}
 {bone1Name=dyn_rope08_cr, bone2Name=dyn_rope05_cr, swing1LimitDegrees=40, swing2LimitDegrees=40, twistLimitDegrees=45, twistMotion=Locked}
 {bone1Name=dyn_rope08_lf, bone2Name=dyn_rope06_lf, swing1LimitDegrees=10, swing2LimitDegrees=10, twistLimitDegrees=45, twistMotion=Locked}
 {bone1Name=dyn_rope04_lf, bone2Name=dyn_rope03_lf, swing1LimitDegrees=40, swing2LimitDegrees=80, twistLimitDegrees=45, twistMotion=Locked}
 {bone1Name=dyn_rope08_lf, bone2Name=dyn_rope04_lf, swing1LimitDegrees=100, swing2LimitDegrees=100, twistLimitDegrees=45, twistMotion=Locked}
 {bone1Name=dyn_rope01_rt, bone2Name=base_rope01_rt, swing1LimitDegrees=28, swing2LimitDegrees=45, twistLimitDegrees=45, twistMotion=Locked}
 {
  bone1Name=dyn_sharkFin_rt
  bone2Name=dyn_sharkRoot_cr
  swing1LimitDegrees=45
  swing1Motion=Locked
  swing2LimitDegrees=45
  swing2Motion=Locked
  twistLimitDegrees=45
  twistMotion=Locked
 }
 {
  bone1Name=dyn_sharkFin_lf
  bone2Name=dyn_sharkRoot_cr
  swing1LimitDegrees=45
  swing1Motion=Locked
  swing2LimitDegrees=45
  swing2Motion=Locked
  twistLimitDegrees=45
  twistMotion=Locked
 }
 {
  bone1Name=dyn_sharkRoot_cr
  bone2Name=pack
  swing1LimitDegrees=45
  swing1Motion=Free
  swing2LimitDegrees=45
  swing2Motion=Free
  twistLimitDegrees=45
  twistMotion=Free
 }
 {
  bone1Name=base_rope06_cr
  bone2Name=pack
  swing1LimitDegrees=45
  swing1Motion=Free
  swing2LimitDegrees=45
  swing2Motion=Free
  twistLimitDegrees=45
  twistMotion=Free
 }
 {
  bone1Name=base_rope06_lf
  bone2Name=pack
  swing1LimitDegrees=45
  swing1Motion=Free
  swing2LimitDegrees=45
  swing2Motion=Free
  twistLimitDegrees=45
  twistMotion=Free
 }
 {
  bone1Name=base_rope01_lf
  bone2Name=pack
  swing1LimitDegrees=45
  swing1Motion=Free
  swing2LimitDegrees=45
  swing2Motion=Free
  twistLimitDegrees=45
  twistMotion=Free
 }
 {
  bone1Name=base_rope01_cr
  bone2Name=pack
  swing1LimitDegrees=45
  swing1Motion=Free
  swing2LimitDegrees=45
  swing2Motion=Free
  twistLimitDegrees=45
  twistMotion=Free
 }
 {
  bone1Name=base_rope01_rt
  bone2Name=pack
  swing1LimitDegrees=45
  swing1Motion=Free
  swing2LimitDegrees=45
  swing2Motion=Free
  twistLimitDegrees=45
  twistMotion=Free
 }
]
