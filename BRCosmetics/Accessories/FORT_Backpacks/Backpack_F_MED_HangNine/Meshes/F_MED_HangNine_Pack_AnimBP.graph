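# Animation Blueprint for the HangNine backpack cosmetic, adding secondary physics motion to the pack and its shark parts.
# The input pose gets an additive pack-bone offset, then runs through a RigidBody node and chained spring-driven AnimDynamics on shark root and fins.
# RigidBody gravity comes from a shared parameter-asset function, or fixed downward gravity during crouch, swim, emote and DBNO states.
# Zipline events set a vector that no shown node reads.
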AnimBlueprint /BRCosmetics/Accessories/FORT_Backpacks/Backpack_F_MED_HangNine/Meshes/F_MED_HangNine_Pack_AnimBP
target_skeleton=&/F_MED_HangNine_Pack_Skeleton
graphs [
 AnimGraph {nodes [
   n1 = Input Pose()
   n2 = RigidBody!(
    ComponentPose="Transform (Modify) Bone - Bone: pack"(ComponentPose=Local To Component(LocalPose=@n1), Translation="0,0.9,0", Rotation="0,0,0") {translationMode=BMM_Additive, translationSpace=BCS_BoneSpace, boneToModify {boneName=pack}, translation {y=0.9}}
    Alpha
    OverrideWorldGravity:Vector=Get GravityOverride()
    ExternalForce
   ) -> (Pose) {
    bOverrideWorldGravity=true
    overridePhysicsAsset=&/F_MED_HangNine_Pack_Physics
    simulationSpace=BaseBoneSpace
    baseBoneRef {boneName=root}
    componentAppliedLinearAccClamp=<4000 4000 4000>
    componentLinearAccScale=<0.3 0.3 0.3>
    componentLinearVelScale=<0.3 0.3 0.3>
    simSpaceSettings {
     maxAngularAcceleration=100
     maxAngularVelocity=400
     maxLinearAcceleration=100
     maxLinearVelocity=400
     velocityScaleZ=0.8
     worldAlpha=0.3
     externalLinearDragV=<0.2 0.2>
    }
   }
   n3 = Anim Dynamics - Bone: dyn_sharkRoot_cr!(
    ComponentPose=@n2
    Alpha=1
    GravityOverride:Vector="0,0,0"
    ComponentLinearAccScale:Vector="0.5,0.5,0.5"
    ComponentLinearVelScale:Vector="0.5,0.5,0.5"
    ExternalForce="0,0,0"
   ) -> (Pose) {
    angularDampingOverride=1
    angularSpringConstant=17500
    bAngularSpring=true
    bOverrideAngularDamping=true
    bUseGravityOverride=true
    gravityScale=0.35
    simulationSpace=RootRelative
    boundBone {boneName=dyn_sharkRoot_cr}
    componentAppliedLinearAccClamp=<3000 3000 3000>
    componentLinearAccScale=<0.5 0.5 0.5>
    componentLinearVelScale=<0.5 0.5 0.5>
    gravityOverride {z=-980}
    physicsBodyDefinitions [{
      boundBone {boneName=dyn_sharkRoot_cr}
      boxExtents=<25 25 25>
      constraintSetup {angularLimitsMax=<5 5 3>, angularLimitsMin=<-5 -5 -5>, angularTarget=<1 0 0>}
      localJointOffset=<1.5 0 0>
     }]
    simSpaceSettings {maxAngularAcceleration=35, maxAngularVelocity=8, simSpaceAngularAlpha=1}
   }
   n4 = Anim Dynamics - Bone: dyn_sharkFin_rt!(
    ComponentPose=@n3
    Alpha=1
    GravityOverride:Vector="0,0,0"
    ComponentLinearAccScale:Vector="0.5,0.5,0.5"
    ComponentLinearVelScale:Vector="0.5,0.5,0.5"
    ExternalForce="0,0,0"
   ) -> (Pose) {
    angularDampingOverride=1
    angularSpringConstant=14000
    bAngularSpring=true
    bOverrideAngularDamping=true
    bUseGravityOverride=true
    gravityScale=0.35
    simulationSpace=RootRelative
    boundBone {boneName=dyn_sharkFin_rt}
    componentAppliedLinearAccClamp=<3000 3000 3000>
    componentLinearAccScale=<0.5 0.5 0.5>
    componentLinearVelScale=<0.5 0.5 0.5>
    gravityOverride {z=-980}
    physicsBodyDefinitions [{
      boundBone {boneName=dyn_sharkFin_rt}
      boxExtents=<14 14 14>
      constraintSetup {angularLimitsMax=<0 8 8>, angularLimitsMin=<0 0 -4>, angularTarget=<1 0 0>}
      localJointOffset=<2 0 0>
     }]
    simSpaceSettings {maxAngularAcceleration=35, maxAngularVelocity=8, simSpaceAngularAlpha=1}
   }
   n5 = Anim Dynamics - Bone: dyn_sharkFin_lf!(
    ComponentPose=@n4
    Alpha=1
    GravityOverride:Vector="0,0,0"
    ComponentLinearAccScale:Vector="0.5,0.5,0.5"
    ComponentLinearVelScale:Vector="0.5,0.5,0.5"
    ExternalForce="0,0,0"
   ) -> (Pose) {
    angularDampingOverride=1
    angularSpringConstant=14000
    bAngularSpring=true
    bOverrideAngularDamping=true
    bUseGravityOverride=true
    gravityScale=0.35
    simulationSpace=RootRelative
    boundBone {boneName=dyn_sharkFin_lf}
    componentAppliedLinearAccClamp=<3000 3000 3000>
    componentLinearAccScale=<0.5 0.5 0.5>
    componentLinearVelScale=<0.5 0.5 0.5>
    gravityOverride {z=-980}
    physicsBodyDefinitions [{
      boundBone {boneName=dyn_sharkFin_lf}
      boxExtents=<14 14 14>
      constraintSetup {angularLimitsMax=<0 10 8>, angularLimitsMin=<0 0 -4>, angularTarget=<1 0 0>}
      localJointOffset=<2 0 0>
     }]
    simSpaceSettings {maxAngularAcceleration=35, maxAngularVelocity=8, simSpaceAngularAlpha=1}
   }
   n6 = Anim Dynamics - Bone: dyn_topfin_cr!(
    ComponentPose=@n5
    Alpha=1
    GravityOverride:Vector="0,0,0"
    ComponentLinearAccScale:Vector="0.5,0.5,0.5"
    ComponentLinearVelScale:Vector="0.5,0.5,0.5"
    ExternalForce="0,0,0"
   ) -> (Pose) {
    angularDampingOverride=1
    angularSpringConstant=14000
    bAngularSpring=true
    bOverrideAngularDamping=true
    bUseGravityOverride=true
    gravityScale=0.35
    simulationSpace=RootRelative
    boundBone {boneName=dyn_topfin_cr}
    componentAppliedLinearAccClamp=<3000 3000 3000>
    componentLinearAccScale=<0.5 0.5 0.5>
    componentLinearVelScale=<0.5 0.5 0.5>
    gravityOverride {z=-980}
    physicsBodyDefinitions [{
      boundBone {boneName=dyn_topfin_cr}
      boxExtents=<14 14 14>
      constraintSetup {angularLimitsMax=<0 5 5>, angularLimitsMin=<0 -5 -5>, angularTarget=<1 0 0>}
      localJointOffset=<2 0 0>
     }]
    simSpaceSettings {maxAngularAcceleration=35, maxAngularVelocity=8, simSpaceAngularAlpha=1}
   }
   Output Pose(Result=Component To Local(ComponentPose=@n6)) {layerGroup=None}
  ]}
 EventGraph {nodes [
   Event BlueprintUpdateAnimation()
   > n1 = "Gravity Override State Function (Parameter Asset) (Nativized)"(
    CharacterPartAnimInstance=Self-Reference()
    GameTimeInSeconds=GetGameTimeInSeconds()
    Parameters=/BRCosmetics/Characters/Player/Female/Medium/Bodies/F_MED_HangNine/Meshes/Parts/F_MED_HangNine_Parameters
   )
   > n2 = Branch(Condition=Get GravOver() -> (GravOver:bool))
   > Set GravityOverride(GravityOverride="0,0,-980")
   Set GravityOverride(execute=@n2.else, GravityOverride=@n1)
   n3 = Event OnBeginCrouching()
   Set GravOver(execute=[@n3.then, @n4.then, @n5.then, @n13.then, @n16.then, @n18.then], GravOver:bool=true) -> (Output_Get:bool)
   n4 = Event OnBeginSwimming() {is=title K2Node_Event}
   n5 = Event OnBeginSwimSprinting() {is=title K2Node_Event}
   n6 = Event OnEndSwimming() {is=title K2Node_Event}
   Set GravOver(execute=[@n6.then, @n7.then, @n14.then, @n15.then, @n17.then, @n19.then], GravOver=false)
   n7 = Event OnEndSwimSprinting() {is=title K2Node_Event}
   n8 = Event BlueprintProcessAnimInputEvent()
   n9 = IsNamed(self=@n8.AnimInputEvent, InEventName=aie.Player.Zipline.Entered)
   n10 = OR Boolean(A=Get bIsMoving2D(), B=Get bIsTacticalSprinting())
   n11 = Get bIsJumping()
   > n12 = Branch(Condition=OR Boolean(A=@n9, B=AND Boolean!(A=@n9, B=@n11, C:bool=@n10) -> (ReturnValue)))
   > Set animDyn on zipline(animDyn on zipline:Vector="0, 0, 0") -> (Output_Get:Vector)
   Branch(execute=@n12.else, Condition=IsNamed(self=@n8.AnimInputEvent, InEventName=aie.Player.Zipline.Exited))
   > Set animDyn on zipline(animDyn on zipline="0.5,0.5,0.5")
   n13 = Event OnBeginPlayingEmote() {is=title K2Node_Event}
   n14 = Event OnEndPlayingEmote() {is=title K2Node_Event}
   n15 = Event OnEndCrouching()
   n16 = Event OnBeginDBNO() {is=title K2Node_Event}
   n17 = Event OnEndDBNO() {is=title K2Node_Event}
   n18 = Event OnBeginDBNOCarried() {is=title K2Node_Event}
   n19 = Event OnEndDBNOCarried() {is=title K2Node_Event}
   TryGetPawnOwner()
  ]}
]
variables [
 animDyn on zipline {type=Vector}
 Bounce {type=number, default=16000}
 GravityOverride {type=Vector}
 GravOver {type=boolean}
]
